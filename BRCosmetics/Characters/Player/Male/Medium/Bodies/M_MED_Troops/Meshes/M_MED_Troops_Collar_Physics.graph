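# Physics asset for the collar of the medium male M_MED_Troops player body, defining capsule bodies and constraints for its dynamic collar bones.
# Three-bone collar chains hang from each clavicle with tight angular limits, except the right tip, which is free and has a smaller, differently oriented capsule.
# Larger clavicle and upper arm capsules cover the shoulders, linked by free-rotation constraints through the spine to the pelvis, with a gun base body on spine_04.
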PhysicsAsset /BRCosmetics/Characters/Player/Male/Medium/Bodies/M_MED_Troops/Meshes/M_MED_Troops_Collar_Physics
bodies {
 clavicle_l [{shapeType=Capsule, length=8.25, radius=7.25, shapeName=clavicle_l_capsule, center=<6.3326 0 0>, rotation=<rot 90 0 0>}]
 clavicle_r [{shapeType=Capsule, length=8.25, radius=7.25, shapeName=clavicle_r_capsule, center=<-5.6834 -2.4392 0.8783>, rotation=<rot 90 0 0>}]
 dyn_collar_upper_01_l [{shapeType=Capsule, length=2.6209, radius=1.7278, shapeName=dyn_shirt_fl_01_capsule, center=<1 0 0>, rotation=<rot 0 0 90>}]
 dyn_collar_upper_01_r [{shapeType=Capsule, length=2.6209, radius=1.7278, shapeName=dyn_shirt_fl_01_capsule, center=<1 0 0>, rotation=<rot 0 0 90>}]
 dyn_collar_upper_02_l [{shapeType=Capsule, length=2.8312, radius=1.5523, shapeName=dyn_shirt_fl_02_capsule, center=<1 0 0>, rotation=<rot 0 0 90>}]
 dyn_collar_upper_02_r [{shapeType=Capsule, length=2.8312, radius=1.5523, shapeName=dyn_shirt_fl_02_capsule, center=<1 0 0>, rotation=<rot 0 0 90>}]
 dyn_collar_upper_03_l [{shapeType=Capsule, length=3.4352, radius=1.5545, shapeName=dyn_shirt_fl_03_capsule, center=<1 0 0>, rotation=<rot 0 0 90>}]
 dyn_collar_upper_03_r [{shapeType=Capsule, length=1.5, radius=1, shapeName=dyn_shirt_fl_03_capsule, center=<1 0 0>, rotation=<rot 90 0 0>}]
 dyn_gun_base [{shapeType=Capsule, length=0.505, radius=0.505, shapeName=dyn_shirt_flap_base_capsule, rotation=<rot 0 -90 0>}]
 pelvis [{shapeType=Capsule, length=1, radius=1, shapeName=pelvis_capsule}]
 spine_02 [{shapeType=Capsule, length=1, radius=1, shapeName=calf_l_capsule}]
 spine_03 [{shapeType=Capsule, length=0.505, radius=0.505, shapeName=dyn_shirt_flap_base_capsule, rotation=<rot 0 -90 0>}]
 spine_04 [{shapeType=Capsule, length=0.505, radius=0.505, shapeName=dyn_shirt_flap_base_capsule, rotation=<rot 0 -90 0>}]
 spine_05 [{shapeType=Capsule, length=1, radius=1, shapeName=spine_05_capsule}]
 upperarm_l [{shapeType=Capsule, length=11.2384, radius=4.3008, shapeName=upperarm_l_capsule, rotation=<rot 90 0 0>}]
 upperarm_r [{shapeType=Capsule, length=8.2594, radius=6.1832, shapeName=upperarm_r_capsule, center=<-3.7561 0 1.984>, rotation=<rot 90 0 0>}]
}
constraints [
 {
  bone1Name=spine_02
  bone2Name=pelvis
  swing1LimitDegrees=45
  swing1Motion=Free
  swing2LimitDegrees=45
  swing2Motion=Free
  twistLimitDegrees=45
  twistMotion=Free
 }
 {
  bone1Name=dyn_gun_base
  bone2Name=spine_04
  swing1LimitDegrees=45
  swing1Motion=Free
  swing2LimitDegrees=45
  swing2Motion=Free
  twistLimitDegrees=45
  twistMotion=Free
 }
 {
  bone1Name=spine_03
  bone2Name=spine_02
  swing1LimitDegrees=45
  swing1Motion=Free
  swing2LimitDegrees=45
  swing2Motion=Free
  twistLimitDegrees=45
  twistMotion=Free
 }
 {
  bone1Name=spine_04
  bone2Name=spine_03
  swing1LimitDegrees=45
  swing1Motion=Free
  swing2LimitDegrees=45
  swing2Motion=Free
  twistLimitDegrees=45
  twistMotion=Free
 }
 {
  bone1Name=spine_05
  bone2Name=spine_04
  swing1LimitDegrees=45
  swing1Motion=Free
  swing2LimitDegrees=45
  swing2Motion=Free
  twistLimitDegrees=45
  twistMotion=Free
 }
 {bone1Name=dyn_collar_upper_01_l, bone2Name=clavicle_l, swing1LimitDegrees=10, swing2LimitDegrees=10, twistLimitDegrees=5}
 {bone1Name=dyn_collar_upper_02_l, bone2Name=dyn_collar_upper_01_l, swing1LimitDegrees=10, swing2LimitDegrees=10, twistLimitDegrees=5}
 {bone1Name=dyn_collar_upper_03_l, bone2Name=dyn_collar_upper_02_l, swing1LimitDegrees=10, swing2LimitDegrees=10, twistLimitDegrees=5}
 {bone1Name=dyn_collar_upper_01_r, bone2Name=clavicle_r, swing1LimitDegrees=10, swing2LimitDegrees=10, twistLimitDegrees=5}
 {bone1Name=dyn_collar_upper_02_r, bone2Name=dyn_collar_upper_01_r, swing1LimitDegrees=10, swing2LimitDegrees=10, twistLimitDegrees=5}
 {
  bone1Name=dyn_collar_upper_03_r
  bone2Name=dyn_collar_upper_02_r
  swing1LimitDegrees=45
  swing1Motion=Free
  swing2LimitDegrees=45
  swing2Motion=Free
  twistLimitDegrees=45
  twistMotion=Free
 }
 {
  bone1Name=clavicle_l
  bone2Name=spine_05
  swing1LimitDegrees=45
  swing1Motion=Free
  swing2LimitDegrees=45
  swing2Motion=Free
  twistLimitDegrees=45
  twistMotion=Free
 }
 {
  bone1Name=clavicle_r
  bone2Name=spine_05
  swing1LimitDegrees=45
  swing1Motion=Free
  swing2LimitDegrees=45
  swing2Motion=Free
  twistLimitDegrees=45
  twistMotion=Free
 }
 {
  bone1Name=upperarm_l
  bone2Name=clavicle_l
  swing1LimitDegrees=45
  swing1Motion=Free
  swing2LimitDegrees=45
  swing2Motion=Free
  twistLimitDegrees=45
  twistMotion=Free
 }
 {
  bone1Name=upperarm_r
  bone2Name=clavicle_r
  swing1LimitDegrees=45
  swing1Motion=Free
  swing2LimitDegrees=45
  swing2Motion=Free
  twistLimitDegrees=45
  twistMotion=Free
 }
]
