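# Physics asset for the Troops male medium body's pants, setting up secondary motion on dynamic pant bones at the left calf.
# Three chains (a, b, c) of three capsule bodies each hang from calf_l, linked by constraints with very tight swing and twist limits for slight movement.
# The calf_l and pelvis bodies are small placeholder-sized capsules, joined by a constraint with free swing and twist.
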PhysicsAsset /BRCosmetics/Characters/Player/Male/Medium/Bodies/M_MED_Troops/Meshes/M_MED_Troops_Pant_Physics
bodies {
 calf_l [{shapeType=Capsule, length=1, radius=1, shapeName=calf_l_capsule}]
 dyn_pant_a_01 [{shapeType=Capsule, length=4.7936, radius=4.0036, shapeName=dyn_pant_a_01_capsule, center=<6.1896 0 0>, rotation=<rot -90 0 0>}]
 dyn_pant_a_02 [{shapeType=Capsule, length=4.9612, radius=4.5196, shapeName=dyn_pant_a_02_capsule, center=<6.1896 0 0>, rotation=<rot -90 0 0>}]
 dyn_pant_a_03 [{shapeType=Capsule, length=0.505, radius=0.505, shapeName=dyn_pant_a_03_capsule, center=<6.1896 0 0>, rotation=<rot -90 0 0>}]
 dyn_pant_b_01 [{shapeType=Capsule, length=6.1721, radius=5.281, shapeName=dyn_pant_b_01_capsule, center=<5.5707 0 0>, rotation=<rot -90 0 0>}]
 dyn_pant_b_02 [{shapeType=Capsule, length=5.6091, radius=4.9634, shapeName=dyn_pant_b_02_capsule, center=<5.5707 0 0>, rotation=<rot -90 0 0>}]
 dyn_pant_b_03 [{shapeType=Capsule, length=0.505, radius=0.505, shapeName=dyn_pant_b_03_capsule, center=<3.3188 0 0>, rotation=<rot -90 0 0>}]
 dyn_pant_c_01 [{shapeType=Capsule, length=4.2171, radius=1.6592, shapeName=dyn_pant_c_01_capsule, center=<2.1381 0 0>, rotation=<rot -90 0 0>}]
 dyn_pant_c_02 [{shapeType=Capsule, length=7.7914, radius=5.2498, shapeName=dyn_pant_c_02_capsule, center=<2.1381 0 0>, rotation=<rot -90 0 0>}]
 dyn_pant_c_03 [{shapeType=Capsule, length=0.505, radius=0.505, shapeName=dyn_pant_c_03_capsule, center=<1.8354 0 0>, rotation=<rot -90 0 0>}]
 pelvis [{shapeType=Capsule, length=1, radius=1, shapeName=pelvis_capsule}]
}
constraints [
 {bone1Name=dyn_pant_a_01, bone2Name=calf_l, swing1LimitDegrees=2.5, swing2LimitDegrees=2.5, twistLimitDegrees=2.5}
 {bone1Name=dyn_pant_a_02, bone2Name=dyn_pant_a_01, swing1LimitDegrees=2.5, swing2LimitDegrees=2.5, twistLimitDegrees=2.5}
 {bone1Name=dyn_pant_a_03, bone2Name=dyn_pant_a_02, swing1LimitDegrees=5, swing2LimitDegrees=5, twistLimitDegrees=5}
 {bone1Name=dyn_pant_b_01, bone2Name=calf_l, swing1LimitDegrees=2.5, swing2LimitDegrees=2.5, twistLimitDegrees=2.5}
 {bone1Name=dyn_pant_b_02, bone2Name=dyn_pant_b_01, swing1LimitDegrees=2.5, swing2LimitDegrees=2.5, twistLimitDegrees=2.5}
 {bone1Name=dyn_pant_b_03, bone2Name=dyn_pant_b_02, swing1LimitDegrees=2.5, swing2LimitDegrees=2.5, twistLimitDegrees=2.5}
 {bone1Name=dyn_pant_c_01, bone2Name=calf_l, swing1LimitDegrees=2.5, swing2LimitDegrees=2.5, twistLimitDegrees=2.5}
 {bone1Name=dyn_pant_c_02, bone2Name=dyn_pant_c_01, swing1LimitDegrees=2.5, swing2LimitDegrees=2.5, twistLimitDegrees=2.5}
 {bone1Name=dyn_pant_c_03, bone2Name=dyn_pant_c_02, swing1LimitDegrees=5, swing2LimitDegrees=5, twistLimitDegrees=5}
 {
  bone1Name=calf_l
  bone2Name=pelvis
  swing1LimitDegrees=45
  swing1Motion=Free
  swing2LimitDegrees=45
  swing2Motion=Free
  twistLimitDegrees=45
  twistMotion=Free
 }
]
