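# Physics asset for the back shirt piece of the medium male Troops character body, giving it a jointed chain for physical motion.
# Four small capsule bodies, shirt_back_01 to 04, link in sequence from spine_01 with narrow swing and tighter twist limits.
# Larger capsules on the pelvis, spine_01 and left thigh form the body volumes, with spine and thigh joined to the pelvis by free angular constraints.
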PhysicsAsset /BRCosmetics/Characters/Player/Male/Medium/Bodies/M_MED_Troops/Meshes/M_MED_Troops_Shirt_Back_Physics
bodies {
 pelvis [{shapeType=Capsule, length=12, radius=8.755, shapeName=spine_01_capsule}]
 shirt_back_01 [{shapeType=Capsule, length=2.6209, radius=1.7278, shapeName=dyn_shirt_fl_01_capsule, center=<2.8823 0 0>, rotation=<rot 0 0.0001 -35>}]
 shirt_back_02 [{shapeType=Capsule, length=2.8312, radius=1.5523, shapeName=dyn_shirt_fl_02_capsule, center=<2.8823 0 0>, rotation=<rot 0 0.0001 -35>}]
 shirt_back_03 [{shapeType=Capsule, length=3.4352, radius=1.5545, shapeName=dyn_shirt_fl_03_capsule, center=<2.8823 0 0>, rotation=<rot 0 0.0001 -35>}]
 shirt_back_04 [{shapeType=Capsule, length=0.505, radius=0.505, shapeName=dyn_shirt_fl_04_capsule, center=<2.284 0 0>}]
 spine_01 [{shapeType=Capsule, length=9.25, radius=8.005, shapeName=spine_01_capsule}]
 thigh_l [{shapeType=Capsule, length=13.095, radius=9.4182, shapeName=thigh_l_capsule, center=<-7.4469 0 0>, rotation=<rot 90 0 0>}]
}
constraints [
 {
  bone1Name=spine_01
  bone2Name=pelvis
  swing1LimitDegrees=45
  swing1Motion=Free
  swing2LimitDegrees=45
  swing2Motion=Free
  twistLimitDegrees=45
  twistMotion=Free
 }
 {
  bone1Name=thigh_l
  bone2Name=pelvis
  swing1LimitDegrees=45
  swing1Motion=Free
  swing2LimitDegrees=45
  swing2Motion=Free
  twistLimitDegrees=45
  twistMotion=Free
 }
 {bone1Name=shirt_back_01, bone2Name=spine_01, swing1LimitDegrees=15, swing2LimitDegrees=15, twistLimitDegrees=5}
 {bone1Name=shirt_back_02, bone2Name=shirt_back_01, swing1LimitDegrees=15, swing2LimitDegrees=15, twistLimitDegrees=5}
 {bone1Name=shirt_back_03, bone2Name=shirt_back_02, swing1LimitDegrees=15, swing2LimitDegrees=15, twistLimitDegrees=5}
 {bone1Name=shirt_back_04, bone2Name=shirt_back_03, swing1LimitDegrees=15, swing2LimitDegrees=15, twistLimitDegrees=5}
]
