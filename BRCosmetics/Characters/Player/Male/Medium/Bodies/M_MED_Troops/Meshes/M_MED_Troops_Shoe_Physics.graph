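# Physics asset for the M_MED_Troops male player cosmetic's shoe mesh, defining a three-bone boot chain on the right leg.
# Boot capsules chain from foot_r with limits that loosen toward the tip; the first two joints lock swing2.
# Pelvis, thigh, calf and foot carry unit-sized capsules joined by free angular constraints.
# Boot capsule shape names carry dyn_pant_c names, unlike their bones.
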PhysicsAsset /BRCosmetics/Characters/Player/Male/Medium/Bodies/M_MED_Troops/Meshes/M_MED_Troops_Shoe_Physics
bodies {
 boot_01 [{shapeType=Capsule, length=4.2171, radius=1.6592, shapeName=dyn_pant_c_01_capsule, center=<3.6482 0 0>, rotation=<rot -90 0 0>}]
 boot_02 [{shapeType=Capsule, length=2.2914, radius=2.7498, shapeName=dyn_pant_c_02_capsule, center=<4.3717 0 0>, rotation=<rot -90 0 0>}]
 boot_03 [{shapeType=Capsule, length=0.505, radius=0.505, shapeName=dyn_pant_c_03_capsule, center=<1.8354 0 0>, rotation=<rot -90 0 0>}]
 calf_r [{shapeType=Capsule, length=1, radius=1, shapeName=calf_r_capsule}]
 foot_r [{shapeType=Capsule, length=1, radius=1, shapeName=foot_r_capsule}]
 pelvis [{shapeType=Capsule, length=1, radius=1, shapeName=pelvis_capsule}]
 thigh_r [{shapeType=Capsule, length=1, radius=1, shapeName=thigh_r_capsule}]
}
constraints [
 {bone1Name=boot_01, bone2Name=foot_r, swing1LimitDegrees=5, swing2LimitDegrees=5, swing2Motion=Locked, twistLimitDegrees=2}
 {bone1Name=boot_02, bone2Name=boot_01, swing1LimitDegrees=10, swing2LimitDegrees=10, swing2Motion=Locked, twistLimitDegrees=2}
 {bone1Name=boot_03, bone2Name=boot_02, swing1LimitDegrees=15, swing2LimitDegrees=15, twistLimitDegrees=5}
 {
  bone1Name=foot_r
  bone2Name=calf_r
  swing1LimitDegrees=45
  swing1Motion=Free
  swing2LimitDegrees=45
  swing2Motion=Free
  twistLimitDegrees=45
  twistMotion=Free
 }
 {
  bone1Name=calf_r
  bone2Name=thigh_r
  swing1LimitDegrees=45
  swing1Motion=Free
  swing2LimitDegrees=45
  swing2Motion=Free
  twistLimitDegrees=45
  twistMotion=Free
 }
 {
  bone1Name=thigh_r
  bone2Name=pelvis
  swing1LimitDegrees=45
  swing1Motion=Free
  swing2LimitDegrees=45
  swing2Motion=Free
  twistLimitDegrees=45
  twistMotion=Free
 }
]
