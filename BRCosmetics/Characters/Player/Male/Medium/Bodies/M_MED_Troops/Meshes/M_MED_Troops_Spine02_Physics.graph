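# Physics asset for the Spine02 variant of the male medium Troops cosmetic body, providing bodies and joints for dynamic shirt flaps.
# Capsules on pelvis, spine_01 and spine_02 form the torso, with two five-bone chains (dyn_shirt_fl and dyn_shirt_flap) branching from spine_02.
# Torso joints and both chain roots rotate freely, while links within each chain have narrow swing and tighter twist limits.
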PhysicsAsset /BRCosmetics/Characters/Player/Male/Medium/Bodies/M_MED_Troops/Meshes/M_MED_Troops_Spine02_Physics
bodies {
 dyn_shirt_fl_01 [{shapeType=Capsule, length=2.1209, radius=2.2278, shapeName=dyn_shirt_fl_01_capsule, center=<2.8823 0 0>, rotation=<rot 0 0.0001 -35>}]
 dyn_shirt_fl_02 [{shapeType=Capsule, length=2.3312, radius=2.0523, shapeName=dyn_shirt_fl_02_capsule, center=<2.8823 0 0>, rotation=<rot 0 0.0001 -35>}]
 dyn_shirt_fl_03 [{shapeType=Capsule, length=2.9352, radius=2.0545, shapeName=dyn_shirt_fl_03_capsule, center=<2.8823 0 0>, rotation=<rot 0 0.0001 -35>}]
 dyn_shirt_fl_04 [{shapeType=Capsule, length=0.505, radius=0.505, shapeName=dyn_shirt_fl_04_capsule, center=<2.284 0 0>}]
 dyn_shirt_fl_base [{shapeType=Capsule, length=0.505, radius=0.505, shapeName=dyn_shirt_fl_base_capsule, rotation=<rot 0 -90 0>}]
 dyn_shirt_flap_01 [{shapeType=Capsule, length=0.505, radius=0.505, shapeName=dyn_shirt_flap_01_capsule, center=<2.284 0 0>}]
 dyn_shirt_flap_02 [{shapeType=Capsule, length=3.665, radius=2.7411, shapeName=dyn_shirt_flap_02_capsule, center=<2.284 0 0>}]
 dyn_shirt_flap_03 [{shapeType=Capsule, length=2.1521, radius=1.6467, shapeName=dyn_shirt_flap_03_capsule, center=<2.284 0 0>}]
 dyn_shirt_flap_04 [{shapeType=Capsule, length=0.505, radius=0.505, shapeName=dyn_shirt_flap_04_capsule, center=<2.284 0 0>}]
 dyn_shirt_flap_base [{shapeType=Capsule, length=0.505, radius=0.505, shapeName=dyn_shirt_flap_base_capsule, rotation=<rot 0 -90 0>}]
 pelvis [{shapeType=Capsule, length=1, radius=1, shapeName=pelvis_capsule}]
 spine_01 [{shapeType=Capsule, length=5.75, radius=10.505, shapeName=spine_01_capsule}]
 spine_02 [{shapeType=Capsule, length=9, radius=9.75, shapeName=calf_l_capsule, rotation=<rot 0 -0.0001 20>}]
}
constraints [
 {
  bone1Name=spine_02
  bone2Name=spine_01
  swing1LimitDegrees=45
  swing1Motion=Free
  swing2LimitDegrees=45
  swing2Motion=Free
  twistLimitDegrees=45
  twistMotion=Free
 }
 {
  bone1Name=dyn_shirt_fl_base
  bone2Name=spine_02
  swing1LimitDegrees=45
  swing1Motion=Free
  swing2LimitDegrees=45
  swing2Motion=Free
  twistLimitDegrees=45
  twistMotion=Free
 }
 {bone1Name=dyn_shirt_fl_01, bone2Name=dyn_shirt_fl_base, swing1LimitDegrees=15, swing2LimitDegrees=15, twistLimitDegrees=5}
 {bone1Name=dyn_shirt_fl_02, bone2Name=dyn_shirt_fl_01, swing1LimitDegrees=15, swing2LimitDegrees=15, twistLimitDegrees=5}
 {bone1Name=dyn_shirt_fl_03, bone2Name=dyn_shirt_fl_02, swing1LimitDegrees=15, swing2LimitDegrees=15, twistLimitDegrees=5}
 {bone1Name=dyn_shirt_fl_04, bone2Name=dyn_shirt_fl_03, swing1LimitDegrees=15, swing2LimitDegrees=15, twistLimitDegrees=5}
 {
  bone1Name=dyn_shirt_flap_base
  bone2Name=spine_02
  swing1LimitDegrees=45
  swing1Motion=Free
  swing2LimitDegrees=45
  swing2Motion=Free
  twistLimitDegrees=45
  twistMotion=Free
 }
 {bone1Name=dyn_shirt_flap_01, bone2Name=dyn_shirt_flap_base, swing1LimitDegrees=15, swing2LimitDegrees=15, twistLimitDegrees=5}
 {bone1Name=dyn_shirt_flap_02, bone2Name=dyn_shirt_flap_01, swing1LimitDegrees=15, swing2LimitDegrees=15, twistLimitDegrees=5}
 {bone1Name=dyn_shirt_flap_03, bone2Name=dyn_shirt_flap_02, swing1LimitDegrees=15, swing2LimitDegrees=15, twistLimitDegrees=5}
 {bone1Name=dyn_shirt_flap_04, bone2Name=dyn_shirt_flap_03, swing1LimitDegrees=15, swing2LimitDegrees=15, twistLimitDegrees=5}
 {
  bone1Name=spine_01
  bone2Name=pelvis
  swing1LimitDegrees=45
  swing1Motion=Free
  swing2LimitDegrees=45
  swing2Motion=Free
  twistLimitDegrees=45
  twistMotion=Free
 }
]
