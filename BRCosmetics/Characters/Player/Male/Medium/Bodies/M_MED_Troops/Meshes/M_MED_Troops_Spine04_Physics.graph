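# Physics asset for the male medium Troops body cosmetic, covering a gun and two straps that hang off spine_04.
# Small capsules on pelvis and spine_02 to spine_04 form an anchor chain with free angular constraints, continuing to dyn_gun_base.
# From that base branch a two-body gun chain and two three-body strap chains.
# These branches use capsule bodies with narrow swing and twist limits, keeping secondary motion tight.
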PhysicsAsset /BRCosmetics/Characters/Player/Male/Medium/Bodies/M_MED_Troops/Meshes/M_MED_Troops_Spine04_Physics
bodies {
 dyn_gun_01 [{shapeType=Capsule, length=3, radius=3, shapeName=dyn_gun_01_capsule, center=<2 0 0>, rotation=<rot 90 0 0>}]
 dyn_gun_02 [{shapeType=Capsule, length=3, radius=3, shapeName=dyn_gun_02_capsule, center=<2 0 0>, rotation=<rot 90 0 0>}]
 dyn_gun_base [{shapeType=Capsule, length=0.505, radius=0.505, shapeName=dyn_shirt_flap_base_capsule, rotation=<rot 0 -90 0>}]
 dyn_strap_a_01 [{shapeType=Capsule, length=1.75, radius=0.75, shapeName=dyn_strap_a_01_capsule, center=<2 0 0>, rotation=<rot -90 0 0>}]
 dyn_strap_a_02 [{shapeType=Capsule, length=1.75, radius=0.755, shapeName=dyn_strap_a_02_capsule, center=<2 0 0>, rotation=<rot -90 0 0>}]
 dyn_strap_a_03 [{shapeType=Capsule, length=1.75, radius=0.755, shapeName=dyn_strap_a_03_capsule, center=<2 0 0>, rotation=<rot -90 0 0>}]
 dyn_strap_b_01 [{shapeType=Capsule, length=1.75, radius=0.75, shapeName=dyn_strap_b_01_capsule, center=<2 0 0>, rotation=<rot -90 0 0>}]
 dyn_strap_b_02 [{shapeType=Capsule, length=1.75, radius=0.755, shapeName=dyn_strap_b_02_capsule, center=<2 0 0>, rotation=<rot -90 0 0>}]
 dyn_strap_b_03 [{shapeType=Capsule, length=1.75, radius=0.755, shapeName=dyn_strap_b_03_capsule, center=<2 0 0>, rotation=<rot -90 0 0>}]
 pelvis [{shapeType=Capsule, length=1, radius=1, shapeName=pelvis_capsule}]
 spine_02 [{shapeType=Capsule, length=1, radius=1, shapeName=calf_l_capsule}]
 spine_03 [{shapeType=Capsule, length=0.505, radius=0.505, shapeName=dyn_shirt_flap_base_capsule, rotation=<rot 0 -90 0>}]
 spine_04 [{shapeType=Capsule, length=0.505, radius=0.505, shapeName=dyn_shirt_flap_base_capsule, rotation=<rot 0 -90 0>}]
}
constraints [
 {
  bone1Name=spine_02
  bone2Name=pelvis
  swing1LimitDegrees=45
  swing1Motion=Free
  swing2LimitDegrees=45
  swing2Motion=Free
  twistLimitDegrees=45
  twistMotion=Free
 }
 {bone1Name=dyn_strap_a_01, bone2Name=dyn_gun_base, swing1LimitDegrees=5, swing2LimitDegrees=5, twistLimitDegrees=5}
 {bone1Name=dyn_strap_a_02, bone2Name=dyn_strap_a_01, swing1LimitDegrees=5, swing2LimitDegrees=5, twistLimitDegrees=5}
 {bone1Name=dyn_strap_a_03, bone2Name=dyn_strap_a_02, swing1LimitDegrees=5, swing2LimitDegrees=5, twistLimitDegrees=5}
 {bone1Name=dyn_strap_b_01, bone2Name=dyn_gun_base, swing1LimitDegrees=5, swing2LimitDegrees=5, twistLimitDegrees=5}
 {bone1Name=dyn_strap_b_02, bone2Name=dyn_strap_b_01, swing1LimitDegrees=5, swing2LimitDegrees=5, twistLimitDegrees=5}
 {bone1Name=dyn_strap_b_03, bone2Name=dyn_strap_b_02, swing1LimitDegrees=5, swing2LimitDegrees=5, twistLimitDegrees=5}
 {bone1Name=dyn_gun_01, bone2Name=dyn_gun_base, swing1LimitDegrees=1.5, swing2LimitDegrees=3.5, twistLimitDegrees=1}
 {
  bone1Name=dyn_gun_base
  bone2Name=spine_04
  swing1LimitDegrees=45
  swing1Motion=Free
  swing2LimitDegrees=45
  swing2Motion=Free
  twistLimitDegrees=45
  twistMotion=Free
 }
 {bone1Name=dyn_gun_02, bone2Name=dyn_gun_01, swing1LimitDegrees=10, swing2LimitDegrees=10, twistLimitDegrees=5}
 {
  bone1Name=spine_03
  bone2Name=spine_02
  swing1LimitDegrees=45
  swing1Motion=Free
  swing2LimitDegrees=45
  swing2Motion=Free
  twistLimitDegrees=45
  twistMotion=Free
 }
 {
  bone1Name=spine_04
  bone2Name=spine_03
  swing1LimitDegrees=45
  swing1Motion=Free
  swing2LimitDegrees=45
  swing2Motion=Free
  twistLimitDegrees=45
  twistMotion=Free
 }
]
