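# Physics asset for the face accessory part of the male medium Troops player cosmetic, covering two dangling hair bone chains and the head.
# Each chain, hair_a and hair_b, links four small capsule bodies in sequence from a larger head capsule, with tight, uniform swing and twist limit angles.
# Capsule shape names reference dyn_shirt_fl, and hair_b_04 lacks the capsule rotation the other hair bodies share.
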PhysicsAsset /BRCosmetics/Characters/Player/Male/Medium/Bodies/M_MED_Troops/Meshes/Parts/M_MED_Troops_FaceAcc_Physics
bodies {
 hair_a_01 [{shapeType=Capsule, length=2.6209, radius=1.7278, shapeName=dyn_shirt_fl_01_capsule, center=<1 0 0>, rotation=<rot 0 0 90>}]
 hair_a_02 [{shapeType=Capsule, length=2.8312, radius=1.5523, shapeName=dyn_shirt_fl_02_capsule, center=<1 0 0>, rotation=<rot 0 0 90>}]
 hair_a_03 [{shapeType=Capsule, length=3.4352, radius=1.5545, shapeName=dyn_shirt_fl_03_capsule, center=<1 0 0>, rotation=<rot 0 0 90>}]
 hair_a_04 [{shapeType=Capsule, length=3.4352, radius=1.5545, shapeName=dyn_shirt_fl_03_capsule, center=<1 0 0>, rotation=<rot 0 0 90>}]
 hair_b_01 [{shapeType=Capsule, length=2.6209, radius=1.7278, shapeName=dyn_shirt_fl_01_capsule, center=<1 0 0>, rotation=<rot 0 0 90>}]
 hair_b_02 [{shapeType=Capsule, length=2.8312, radius=1.5523, shapeName=dyn_shirt_fl_02_capsule, center=<1 0 0>, rotation=<rot 0 0 90>}]
 hair_b_03 [{shapeType=Capsule, length=3.4352, radius=1.5545, shapeName=dyn_shirt_fl_03_capsule, center=<1 0 0>, rotation=<rot 0 0 90>}]
 hair_b_04 [{shapeType=Capsule, length=3.4352, radius=1.5545, shapeName=dyn_shirt_fl_03_capsule, center=<1 0 0>}]
 head [{shapeType=Capsule, length=12.25, radius=6.5, rotation=<rot -90 9.4623 40.5373>}]
}
constraints [
 {bone1Name=hair_a_01, bone2Name=head, swing1LimitDegrees=5, swing2LimitDegrees=5, twistLimitDegrees=5}
 {bone1Name=hair_a_02, bone2Name=hair_a_01, swing1LimitDegrees=5, swing2LimitDegrees=5, twistLimitDegrees=5}
 {bone1Name=hair_a_03, bone2Name=hair_a_02, swing1LimitDegrees=5, swing2LimitDegrees=5, twistLimitDegrees=5}
 {bone1Name=hair_a_04, bone2Name=hair_a_03, swing1LimitDegrees=5, swing2LimitDegrees=5, twistLimitDegrees=5}
 {bone1Name=hair_b_01, bone2Name=head, swing1LimitDegrees=5, swing2LimitDegrees=5, twistLimitDegrees=5}
 {bone1Name=hair_b_02, bone2Name=hair_b_01, swing1LimitDegrees=5, swing2LimitDegrees=5, twistLimitDegrees=5}
 {bone1Name=hair_b_03, bone2Name=hair_b_02, swing1LimitDegrees=5, swing2LimitDegrees=5, twistLimitDegrees=5}
 {bone1Name=hair_b_04, bone2Name=hair_b_03, swing1LimitDegrees=5, swing2LimitDegrees=5, twistLimitDegrees=5}
]
